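; Zero Point actor Blueprint whose FX sphere size follows a replicated scale value.
; On replication of that scale, the RepNotify applies it as a uniform world scale to the sphere component, if present.
; It then writes the sphere's resulting bounds radius to ZeroPoint_Radius in the RadishPortal Material Parameter Collection, keeping materials in sync with the sphere, and logs it.
; The construction script is empty.
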(fn OnRep_ZeroPointReplicatedScale
  (bind _sm_fx_radish_zp_sphere (Variables|Default|GetSM_FX_Radish_ZP_Sphere)
    (:then
      (Transformation|SetWorldScale3D _sm_fx_radish_zp_sphere (Math|Conversions|ToVector(Float) (Variables|Default|GetZeroPointReplicatedScale)))
      (Rendering|Material|SetScalarParameterValue "/BlastBerryMapContent/Game/Blueprints/ZP/TechArt/Portals/MPC_RadishPortal.MPC_RadishPortal" "ZeroPoint_Radius" (.SphereRadius (Collision|GetComponentBounds _sm_fx_radish_zp_sphere)) :!class /Script/Engine.KismetMaterialLibrary)
      (Fortnite|FortShippingLog self (Utilities|String|Append "Updated Zero Point Sphere Radus and MPC value to = " (Utilities|String|ToString(Float) (.SphereRadius (Collision|GetComponentBounds _sm_fx_radish_zp_sphere))))))
    (:else)))

(fn ConstructionScript)
