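# Niagara idle effect for the Zero Point, whose look is driven by a User.Stage integer through per-emitter StageSelection inputs.
# Energy-wave meshes, drifting light meshes and glow pulses change size or opacity per stage, while the RainbowCorona003-005 emitters are disabled.
# Looping, distance-culled SolarFlares emitters appear from stage 4, bending ribbons and sprites along Bezier arcs between random points on a stage-sized radius.
# The _Front flare variant anchors its start point around the camera vector.
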NiagaraSystem /BlastBerryMapContent/Game/Blueprints/ZP/NS_Radish_ZeroPoint_Idle
dependencies {
 usedDataInterfaces [&NiagaraDataInterfaceCurve, &NiagaraDataInterfaceCamera]
 usedDynamicInputs [&MakeFloatFromInt, &ReturnParticleIDAttributes]
 usedModules [&SystemState]
}
emitters {
 Empty {
  emitterUpdateScript [
   &EmitterState()
   &SpawnBurst_Instantaneous(Use Spawn Probability=false, UseLoopCountLimit=true)
  ]
  particleSpawnScript [
   &InitializeParticle(
    Color=<rgba 0 0 0 0.35>
    Lifetime=5
    Sprite Size Mode=::NewEnumerator3
    Uniform Sprite Size=9747.168
    UsePositionOffset=true
    Write Material Random=true
   )
   &CameraOffset(Camera Offset Amount=-100)
  ]
  particleUpdateScript [&ParticleState(Kill Particles When Lifetime Has Elapsed=false)]
  rendererClasses [&NiagaraSpriteRendererProperties]
  renderers [{rendererClass=&NiagaraSpriteRendererProperties}]
 }
 EnergyWaves {
  emitterUpdateScript [
   &EmitterState(
    Inactive Response=::NewEnumerator1
    Life Cycle Mode=::NewEnumerator1
    Loop Behavior=::NewEnumerator1
    Loop Duration Mode=::NewEnumerator1
    UseLoopDelay=false
   )
   &SpawnBurst_Instantaneous(Use Spawn Probability=false, UseLoopCountLimit=false)
   &SpawnBurst_Instantaneous(Spawn Time=6.1749, Use Spawn Probability=false, UseLoopCountLimit=false) {enabled=false, module=SpawnBurst_Instantaneous001}
  ]
  particleSpawnScript [
   &InitializeParticle(Color=<rgba 0 0.3261 1 1>, Lifetime=5, Mesh Scale Mode=::NewEnumerator3, UsePositionOffset=false, Write Material Random=true)
   &InitialMeshOrientation(
    Mesh Axis To Orient=<0 0 1>
    Mesh Orientation Mode=::NewEnumerator2
    Rotation=&RandomRangeVector(Minimum=<0 0 0>, Maximum=<0 0 0>)
    Use Rotation Vector=true
   )
  ]
  particleUpdateScript [
   &ParticleState(Kill Particles When Lifetime Has Elapsed=false, Let Infinitely Lived Particles Die When Emitter Deactivates=true)
   &ScaleMeshSize(Scale Factor=&VectorFromFloat(Value=&self:StageSelection(
      Stage:NiagaraInt32=$User.Stage
      Stage 0:NiagaraFloat=0
      Stage 1:NiagaraFloat=7
      Stage 2:NiagaraFloat=10
      Stage 3:NiagaraFloat=16
      Stage 4:NiagaraFloat=18.5
      Stage 5:NiagaraFloat=24
      Stage 6:NiagaraFloat=18.5
     )))
  ]
  rendererClasses [&NiagaraMeshRendererProperties]
  renderers [{rendererClass=&NiagaraMeshRendererProperties, rendererIndex=0}]
 }
 LightMesh {
  emitterUpdateScript [
   &EmitterState(Life Cycle Mode=::NewEnumerator1, Loop Behavior=::NewEnumerator1, Loop Duration Mode=::NewEnumerator1, UseLoopDelay=false)
   &SpawnBurst_Instantaneous(Spawn Count=10, Use Spawn Probability=false, UseLoopCountLimit=false) {enabled=false}
   &SpawnRate(SpawnRate=2, Use Spawn Probability=false)
  ]
  particleSpawnScript [
   &InitializeParticle(
    Color=<rgba 0 0.9782 3 1>
    Lifetime=4
    Mesh Scale=<1.5 1.5 1.5>
    Mesh Scale Mode=::NewEnumerator4
    UsePositionOffset=false
    Write Material Random=true
   )
   &ShapeLocation(
    MaskSpawnGroup=false
    Override Seed=false
    Sphere Radius=&self:StageSelection_01(
     Stage:NiagaraInt32=$User.Stage
     Stage 0:NiagaraFloat=0
     Stage 1:NiagaraFloat=150
     Stage 2:NiagaraFloat=200
     Stage 3:NiagaraFloat=300
     Stage 4:NiagaraFloat=300
     Stage 5:NiagaraFloat=300
     Stage 6:NiagaraFloat=300
    )
    Sphere Surface Distribution=1
    UseOwnerScale=true
   )
   &AddVelocity(Velocity=&Multiply_Vector(A=$Particles.ShapeLocation.ShapeNormal, B=&VectorFromFloat(Value=1)))
   &DynamicMaterialParameters(
    Index 0 Param 1=75
    Index 0 Param 2=120
    Module.Param0WriteEnabled=true
    Module.Param1WriteEnabled=true
    Module.Param2WriteEnabled=true
    Module.Param3WriteEnabled=true
   )
  ]
  particleUpdateScript [
   &ParticleState()
   &ScaleMeshSize(Scale Factor=&MakeVector(
     X=&FloatFromCurve(FloatCurve=DataInterface() {propertyValues {Curve {keys [<at 0 0 RCIM_Cubic>,<at 0.05 1 RCIM_Cubic>]}}})
     Y=1
     Z=1
    ))
   &SolveForcesAndVelocity(Clamp Velocity=false, Limit Acceleration=false)
   &ScaleColor(
    Scale Alpha=&FloatFromCurve(
     FloatCurve=DataInterface() {propertyValues {Curve {keys [
         {interpMode=RCIM_Cubic, tangentMode=RCTM_User, time=0, value=0}
         {interpMode=RCIM_Cubic, tangentMode=RCTM_User, time=0.5, value=1}
         {tangentMode=RCTM_User, time=1, value=0}
        ]}}}
     Scale Curve=&self:StageSelection(
      Stage:NiagaraInt32=$User.Stage
      Stage 0:NiagaraFloat=0
      Stage 1:NiagaraFloat=0.025
      Stage 2:NiagaraFloat=0.25
      Stage 3:NiagaraFloat=0.25
      Stage 4:NiagaraFloat=0.65
      Stage 5:NiagaraFloat=0.8
      Stage 6:NiagaraFloat=1
     )
    )
    ScaleA=true
    ScaleRGB=true
   )
  ]
  rendererClasses [&NiagaraSpriteRendererProperties, &NiagaraMeshRendererProperties]
  renderers [
   {rendererClass=&NiagaraSpriteRendererProperties}
   {rendererClass=&NiagaraMeshRendererProperties, rendererIndex=1}
  ]
 }
 Purple_Mesh_Glow_Burst {
  emitterUpdateScript [
   &EmitterState(
    Enable Distance Culling=true
    MaxDistance=30000
    Module.bUseMaxDistance=true
    Module.bUseMinDistance=false
    Scalability Mode=::NewEnumerator1
   )
   &SpawnBurst_Instantaneous(Spawn Count=1, Use Spawn Probability=false) {enabled=false}
   &SpawnRate(Use Spawn Probability=false)
  ]
  particleSpawnScript [
   /Niagara/Modules/Spawn/Initialization/InitializeParticle(
    Color=<rgba 0 1.3042 4 0.75>
    Lifetime=&RandomRangeFloat(Minimum=3, Maximum=4)
    Mesh Scale=&VectorFromFloat(Value=&self:StageSelection_02(
      Stage:NiagaraInt32=$User.Stage
      Stage 0:NiagaraFloat=0
      Stage 1:NiagaraFloat=0.9
      Stage 2:NiagaraFloat=1.2
      Stage 3:NiagaraFloat=1.5
      Stage 4:NiagaraFloat=2.2
      Stage 5:NiagaraFloat=2.5
      Stage 6:NiagaraFloat=4
     ))
    Module.Write Color=true
    Module.Write Lifetime=true
    Module.Write Mass=true
    Module.Write Material Random=true
    Module.Write Position=true
    Module.Write Scale=true
    Module.Write SpriteRotation=false
    Module.Write SpriteSize=false
    Sprite Rotation=&RandomRangeFloat(Minimum=0, Maximum=360)
    Sprite Size=&Vector2DFromFloat(Value=&RandomRangeFloat(Minimum=3000, Maximum=3200))
   )
   &InitialMeshOrientation(Use Rotation Vector=false)
  ]
  particleUpdateScript [
   &ParticleState()
   &DynamicMaterialParameters(
    Index 0 Param 1=1
    Index 0 Param 2=1
    Index 0 Param 3=1
    Index 0 Param 4=1
    Module.Param0WriteEnabled=true
    Module.Param1WriteEnabled=true
    Module.Param2WriteEnabled=true
    Module.Param3WriteEnabled=true
   )
   &ScaleColor(
    Scale Alpha=&FloatFromCurve(FloatCurve=DataInterface() {propertyValues {Curve {keys [
         {interpMode=RCIM_Cubic, tangentMode=RCTM_User, time=0, value=0}
         {interpMode=RCIM_Cubic, tangentMode=RCTM_User, time=0.5, value=1}
         {tangentMode=RCTM_User, time=1, value=0}
        ]}}})
    ScaleA=true
    ScaleRGB=true
   )
   &ScaleMeshSize(Scale Factor=&VectorFromFloat(Value=&FloatFromCurve(FloatCurve=DataInterface() {propertyValues {Curve {keys [{interpMode=RCIM_Cubic, leaveTangent=1.1268, tangentMode=RCTM_Break, time=0, value=0.5}, <at 1 1>]}}})))
  ]
  rendererClasses [&NiagaraMeshRendererProperties]
  renderers [{rendererClass=&NiagaraMeshRendererProperties, rendererIndex=0}]
 }
 RainbowCorona003 {
  bEnabled=false
  emitterUpdateScript [
   &EmitterState(
    Inactive Response=::NewEnumerator1
    Life Cycle Mode=::NewEnumerator1
    Loop Behavior=::NewEnumerator1
    Loop Duration Mode=::NewEnumerator1
    UseLoopDelay=false
   )
   &SpawnBurst_Instantaneous(Use Spawn Probability=false, UseLoopCountLimit=false)
  ]
  particleSpawnScript [
   /Niagara/Modules/Spawn/Initialization/V2/InitializeParticle(
    Color=<rgba 1 1 1 0.2>
    Lifetime=5
    Mesh Scale Mode=::NewEnumerator3
    Mesh Uniform Scale=20
    Sprite Size Mode=::NewEnumerator3
    Uniform Sprite Size=4000
    UsePositionOffset=false
    Write Material Random=true
   )
   &DynamicMaterialParameters(
    Index 0 Param 1=1500
    Module.Param0WriteEnabled=true
    Module.Param0WriteEnabled1=true
    Module.Param1WriteEnabled=true
    Module.Param1WriteEnabled1=true
    Module.Param2WriteEnabled=true
    Module.Param2WriteEnabled1=true
    Module.Param3WriteEnabled=true
    Module.Param3WriteEnabled1=true
    Write Parameter Index 1=true
   )
   &InitialMeshOrientation(Mesh Axis To Orient=<0 0 1>, Mesh Orientation Mode=::NewEnumerator2, Use Rotation Vector=true)
  ]
  particleUpdateScript [
   &ParticleState(Kill Particles When Lifetime Has Elapsed=false, Let Infinitely Lived Particles Die When Emitter Deactivates=true)
   &ScaleSpriteSize(Scale Sprite Size Mode=::NewEnumerator0, Uniform Scale Factor=$User.SequenceRingSize) {enabled=false}
   &ScaleColor(
    Scale Alpha=&self:StageSelection(
     Stage:NiagaraInt32=$User.Stage
     Stage 0:NiagaraFloat=0.1
     Stage 1:NiagaraFloat=0.025
     Stage 2:NiagaraFloat=0.1
     Stage 3:NiagaraFloat=0.1
     Stage 4:NiagaraFloat=0.3
     Stage 5:NiagaraFloat=0.4
     Stage 6:NiagaraFloat=0.5
    )
    Scale RGB=&VectorFromFloat(Value=$User.SequenceRingIntensity)
    ScaleA=true
    ScaleRGB=true
   )
  ]
  rendererClasses [&NiagaraSpriteRendererProperties, &NiagaraMeshRendererProperties]
  renderers [
   {rendererClass=&NiagaraSpriteRendererProperties}
   {rendererClass=&NiagaraMeshRendererProperties, rendererIndex=1}
  ]
 }
 RainbowCorona004 {
  bEnabled=false
  emitterUpdateScript [
   &EmitterState(
    Inactive Response=::NewEnumerator1
    Life Cycle Mode=::NewEnumerator1
    Loop Behavior=::NewEnumerator1
    Loop Duration Mode=::NewEnumerator1
    UseLoopDelay=false
   )
   &SpawnBurst_Instantaneous(Use Spawn Probability=false, UseLoopCountLimit=false)
  ]
  particleSpawnScript [
   &InitializeParticle(
    Color=<rgba 1 1 1 0.2>
    Lifetime=5
    Mesh Scale Mode=::NewEnumerator3
    Mesh Uniform Scale=60
    Sprite Size Mode=::NewEnumerator3
    Uniform Sprite Size=4000
    UsePositionOffset=false
    Write Material Random=true
   )
   &DynamicMaterialParameters(
    Index 0 Param 1=3000
    Module.Param0WriteEnabled=true
    Module.Param0WriteEnabled1=true
    Module.Param1WriteEnabled=true
    Module.Param1WriteEnabled1=true
    Module.Param2WriteEnabled=true
    Module.Param2WriteEnabled1=true
    Module.Param3WriteEnabled=true
    Module.Param3WriteEnabled1=true
    Write Parameter Index 1=true
   )
   &InitialMeshOrientation(Mesh Axis To Orient=<0 0 1>, Mesh Orientation Mode=::NewEnumerator2, Use Rotation Vector=true)
  ]
  particleUpdateScript [
   &ParticleState(Kill Particles When Lifetime Has Elapsed=false, Let Infinitely Lived Particles Die When Emitter Deactivates=true)
   &ScaleSpriteSize(Scale Sprite Size Mode=::NewEnumerator0, Uniform Scale Factor=$User.SequenceRingSize) {enabled=false}
   &ScaleColor(
    Scale Alpha=&self:StageSelection(
     Stage:NiagaraInt32=$User.Stage
     Stage 0:NiagaraFloat=0.1
     Stage 1:NiagaraFloat=0.025
     Stage 2:NiagaraFloat=0.1
     Stage 3:NiagaraFloat=0.1
     Stage 4:NiagaraFloat=0.3
     Stage 5:NiagaraFloat=0.4
     Stage 6:NiagaraFloat=0.5
    )
    Scale RGB=&VectorFromFloat(Value=$User.SequenceRingIntensity)
    ScaleA=true
    ScaleRGB=true
   )
   &ScaleColor(
    Scale Alpha=&ClampFloat(Float=&FN_LOD_Distance_Remap(Start Distance=4000, End Distance=6000, Minimum Output=0, Maximum Output=1))
    ScaleA=true
    ScaleRGB=true
   ) {module=ScaleColor001}
  ]
  rendererClasses [&NiagaraSpriteRendererProperties, &NiagaraMeshRendererProperties]
  renderers [
   {rendererClass=&NiagaraSpriteRendererProperties}
   {rendererClass=&NiagaraMeshRendererProperties, rendererIndex=1}
  ]
 }
 RainbowCorona005 {
  bEnabled=false
  emitterUpdateScript [
   &EmitterState(
    Inactive Response=::NewEnumerator1
    Life Cycle Mode=::NewEnumerator1
    Loop Behavior=::NewEnumerator1
    Loop Duration Mode=::NewEnumerator1
    UseLoopDelay=false
   )
   &SpawnBurst_Instantaneous(Use Spawn Probability=false, UseLoopCountLimit=false)
  ]
  particleSpawnScript [
   &InitializeParticle(
    Color=<rgba 1 1 1 0.2>
    Lifetime=5
    Mesh Scale Mode=::NewEnumerator3
    Mesh Uniform Scale=140
    Sprite Size Mode=::NewEnumerator3
    Uniform Sprite Size=4000
    UsePositionOffset=false
    Write Material Random=true
   )
   &DynamicMaterialParameters(
    Index 0 Param 1=6000
    Module.Param0WriteEnabled=true
    Module.Param0WriteEnabled1=true
    Module.Param1WriteEnabled=true
    Module.Param1WriteEnabled1=true
    Module.Param2WriteEnabled=true
    Module.Param2WriteEnabled1=true
    Module.Param3WriteEnabled=true
    Module.Param3WriteEnabled1=true
    Write Parameter Index 1=true
   )
   &InitialMeshOrientation(Mesh Axis To Orient=<0 0 1>, Mesh Orientation Mode=::NewEnumerator2, Use Rotation Vector=true)
  ]
  particleUpdateScript [
   &ParticleState(Kill Particles When Lifetime Has Elapsed=false, Let Infinitely Lived Particles Die When Emitter Deactivates=true)
   &ScaleSpriteSize(Scale Sprite Size Mode=::NewEnumerator0, Uniform Scale Factor=$User.SequenceRingSize) {enabled=false}
   &ScaleColor(
    Scale Alpha=&self:StageSelection(
     Stage:NiagaraInt32=$User.Stage
     Stage 0:NiagaraFloat=0.1
     Stage 1:NiagaraFloat=0.025
     Stage 2:NiagaraFloat=0.1
     Stage 3:NiagaraFloat=0.1
     Stage 4:NiagaraFloat=0.3
     Stage 5:NiagaraFloat=0.4
     Stage 6:NiagaraFloat=0.5
    )
    Scale RGB=&VectorFromFloat(Value=$User.SequenceRingIntensity)
    ScaleA=true
    ScaleRGB=true
   )
   &ScaleColor(
    Scale Alpha=&ClampFloat(Float=&FN_LOD_Distance_Remap(Start Distance=6000, End Distance=8000, Minimum Output=0, Maximum Output=1))
    ScaleA=true
    ScaleRGB=true
   ) {module=ScaleColor001}
  ]
  rendererClasses [&NiagaraSpriteRendererProperties, &NiagaraMeshRendererProperties]
  renderers [
   {rendererClass=&NiagaraSpriteRendererProperties}
   {rendererClass=&NiagaraMeshRendererProperties, rendererIndex=1}
  ]
 }
 SolarFlares {
  emitterUpdateScript [
   &EmitterState(
    Enable Distance Culling=true
    Inactive Response=::NewEnumerator1
    Life Cycle Mode=::NewEnumerator1
    Loop Delay=&RandomRangeFloat(Minimum=0.1, Maximum=0.5)
    Loop Duration=&RandomRangeFloat(Minimum=2, Maximum=3)
    MaxDistance=21000
    Module.bUseMaxDistance=true
    Module.bUseMinDistance=false
    Recalculate Duration Each Loop=true
    Scalability Mode=::NewEnumerator1
    UseLoopDelay=true
   )
   &SpawnBurst_Instantaneous(
    Spawn Count=3
    Spawn Probability=&self:StageSelection(
     Stage:NiagaraInt32=$User.Stage
     Stage 0:NiagaraFloat=0
     Stage 1:NiagaraFloat=0
     Stage 2:NiagaraFloat=0
     Stage 3:NiagaraFloat=0
     Stage 4:NiagaraFloat=1
     Stage 5:NiagaraFloat=1
     Stage 6:NiagaraFloat=1
    )
    Use Spawn Probability=true
    UseLoopCountLimit=false
   ) {module=SpawnBurst_Instantaneous001}
   /SetVariables(Emitter.RandomRange:Vector3f=&RandomRangeVector(Minimum=<0 0 0>, Maximum=<1 1 1>, Random Seed=1)) {isSetParameters=true, module=SetVariables_#1}
   /SetVariables(
    Emitter.Radius:NiagaraFloat=&self:StageSelection(
     Stage:NiagaraInt32=$User.Stage
     Stage 0:NiagaraFloat=0
     Stage 1:NiagaraFloat=375
     Stage 2:NiagaraFloat=300
     Stage 3:NiagaraFloat=500
     Stage 4:NiagaraFloat=500
     Stage 5:NiagaraFloat=580
     Stage 6:NiagaraFloat=650
    )
    Emitter.V1:Vector3f=&RandomVector(Vector Scale=1, Constrain to Hemisphere=true, Hemisphere Constraint Vector=$Emitter.RandomRange, Recalculate Random Each Loop=true)
   ) {isSetParameters=true, module=SetVariables_#2}
   /SetVariables(
    Emitter.Radius:NiagaraFloat=&self:StageSelection(
     Stage:NiagaraInt32=$User.Stage
     Stage 0:NiagaraFloat=0
     Stage 1:NiagaraFloat=375
     Stage 2:NiagaraFloat=300
     Stage 3:NiagaraFloat=500
     Stage 4:NiagaraFloat=500
     Stage 5:NiagaraFloat=580
     Stage 6:NiagaraFloat=650
    )
    Emitter.V1:Vector3f=&RandomVector(Vector Scale=1)
   ) {isSetParameters=true}
   /SetVariables(Emitter.V2:Vector3f=&self:ScratchDynamicInput(
     Angle:NiagaraFloat=&Divide_Float(A=180, B=360)
     Rotation Axis:Vector3f=&RandomRangeVector(Minimum=<0 0 0>, Maximum=<1 1 1>)
     Vector To Rotate:Vector3f=$Emitter.V1
    )) {isSetParameters=true}
  ]
  particleSpawnScript [
   &InitializeParticle(
    Color=<rgba 0 1.9563 6 1>
    Lifetime=$Emitter.CurrentLoopDuration
    Ribbon Width=55
    Ribbon Width Mode=::NewEnumerator1
    Sprite Size Mode=::NewEnumerator3
    Uniform Sprite Size=100
    UsePositionOffset=false
    Write Material Random=true
   )
   &PartitionParticles(
    Normalized Index In Partition Includes One=true
    Number Of Partitions=3
    Particles Per Partition=1
    Partition Mode:ENiagara_PartitionMode=::NewEnumerator1
   )
   /SetVariables(Particles.RibbonLinkOrder:NiagaraFloat=&ReturnNormalizedExecIndex()) {isSetParameters=true, module=SetVariables_#3}
   /SetVariables(
    Particles.EndPosition:Vector3f=&Add_Vector(
     A=/Niagara/DynamicInputs/Helpers/SimulationPosition()
     B=&Multiply_VectorByFloat(Vector=$Emitter.V2, Float=$Emitter.Radius)
    )
    Particles.StartPosition:Vector3f=&Add_Vector(
     A=&SimulationPosition()
     B=&Multiply_VectorByFloat(Vector=$Emitter.V1, Float=$Emitter.Radius)
    )
   ) {isSetParameters=true, module=SetVariables_#4}
   /SetVariables(Particles.MiddlePoint:Vector3f=&Lerp_Vector(A=$Particles.StartPosition, B=$Particles.EndPosition, Alpha=0.5)) {isSetParameters=true, module=SetVariables_#5}
   &CameraOffset(Camera Offset Amount=30)
   &DynamicMaterialParameters(
    Index 0 Param 1=2
    Module.Param0WriteEnabled=true
    Module.Param1WriteEnabled=true
    Module.Param2WriteEnabled=true
    Module.Param3WriteEnabled=true
   )
  ]
  particleUpdateScript [
   &ParticleState()
   /SetVariables(Particles.FlareDistance:NiagaraFloat=&FloatFromCurve(
     FloatCurve=DataInterface() {propertyValues {Curve {keys [{arriveTangent=2.0921, interpMode=RCIM_Cubic, leaveTangent=2.0921, tangentMode=RCTM_User, time=0, value=0}, <at 1 1 RCIM_Cubic>]}}}
     Scale Curve=&RandomRangeFloat(Minimum=700, Maximum=400)
    )) {isSetParameters=true, module=SetVariables_#6}
   &self:BezierSpline(
    EndHandle:Vector3f=&Add_Vector(
     A=&Lerp_Vector(A=$Particles.MiddlePoint, B=$Particles.EndPosition, Alpha=0.5)
     B=&Multiply_VectorByFloat(
      Vector=&NormalizeVector(
       Vector To Normalize=&Subtract_Vector(A=$Particles.MiddlePoint, B=&SimulationPosition())
       Safe Normalize=true
       Safe Fallback Threshold=0
       Safe Fallback Vector=<0 0 0.001>
      )
      Float=$Particles.FlareDistance
     )
    )
    EndPosition:Vector3f=$Particles.EndPosition
    PositionAlongSpline:NiagaraFloat=&ReturnNormalizedExecIndex()
    StartHandle:Vector3f=&Add_Vector(
     A=&Lerp_Vector(A=$Particles.EndPosition, B=$Particles.MiddlePoint, Alpha=0.5)
     B=&Multiply_VectorByFloat(
      Vector=&NormalizeVector(
       Vector To Normalize=&Subtract_Vector(A=$Particles.MiddlePoint, B=&SimulationPosition())
       Safe Normalize=true
       Safe Fallback Threshold=0
       Safe Fallback Vector=<0 0 0.001>
      )
      Float=$Particles.FlareDistance
     )
    )
    StartPosition:Vector3f=$Particles.StartPosition
   ) {module=BezierSpline}
   /SetVariables(Particles.Position:Vector3f=$Particles.Spline_Position) {isSetParameters=true, module=SetVariables_#7}
   &Color(Color=&LerpLinearColors(StartColor=<rgba 0 0.019 1 1>, EndColor=<rgba 1 0.0334 0 1>, LerpFactor=$Particles.PartitionParticles.Normalized Index In Partition)) {enabled=false}
   /SetVariables(Particles.Position=&Lerp_Vector(A=$Particles.StartPosition, B=$Particles.EndPosition, Alpha=&ReturnNormalizedExecIndex())) {enabled=false, isSetParameters=true, module=SetVariables_#8}
   &ScaleColor(
    Scale Alpha=&FloatFromCurve(FloatCurve=DataInterface() {propertyValues {Curve {keys [<at 0 0 RCIM_Cubic>,<at 0.25 1 RCIM_Cubic>,<at 1 0 RCIM_Cubic>]}}})
    ScaleA=true
    ScaleRGB=true
   )
   /SetVariables(Particles.SpriteColor=&MakeLinearColorFromVectorAndFloat(
     "Vector (RGB)"=&MakeVectorFromLinearColorRGB(Color=<rgba 0 0.3261 1 1>)
     "Float (Alpha)"=&Multiply_Float(
      A=&MakeCustomFloatFromBool(
       Bool=&SetBoolByIntComparison(A=$Particles.PartitionParticles.Partition, B=1, Comparison Type=::NewEnumerator4)
       True Float=0
       False Float=1
      )
      B=&FloatFromCurve(FloatCurve=DataInterface() {propertyValues {Curve {keys [<at 0 0 RCIM_Cubic>,<at 0.25 1 RCIM_Cubic>,<at 1 0 RCIM_Cubic>]}}})
     )
    )) {isSetParameters=true, module=SetVariables_#9}
  ]
  rendererClasses [&NiagaraSpriteRendererProperties, &NiagaraRibbonRendererProperties]
  renderers [
   {rendererClass=&NiagaraSpriteRendererProperties}
   {rendererClass=&NiagaraRibbonRendererProperties, rendererIndex=1}
   {rendererClass=&NiagaraSpriteRendererProperties, rendererIndex=2}
   {rendererClass=&NiagaraSpriteRendererProperties, rendererIndex=3}
   {rendererClass=&NiagaraRibbonRendererProperties, rendererIndex=4}
   {rendererClass=&NiagaraSpriteRendererProperties, rendererIndex=5}
  ]
 }
 SolarFlares001 {
  emitterUpdateScript [
   &EmitterState(
    Enable Distance Culling=true
    Inactive Response=::NewEnumerator1
    Life Cycle Mode=::NewEnumerator1
    Loop Delay=&RandomRangeFloat(Minimum=0.1, Maximum=0.5)
    Loop Duration=&RandomRangeFloat(Minimum=2.5, Maximum=3.5)
    MaxDistance=21000
    Module.bUseMaxDistance=true
    Module.bUseMinDistance=false
    Recalculate Duration Each Loop=true
    Scalability Mode=::NewEnumerator1
    UseLoopDelay=true
   )
   &SpawnBurst_Instantaneous(
    Spawn Count=3
    Spawn Probability=&self:StageSelection(
     Stage:NiagaraInt32=$User.Stage
     Stage 0:NiagaraFloat=0
     Stage 1:NiagaraFloat=0
     Stage 2:NiagaraFloat=0
     Stage 3:NiagaraFloat=0
     Stage 4:NiagaraFloat=1
     Stage 5:NiagaraFloat=1
     Stage 6:NiagaraFloat=1
    )
    Use Spawn Probability=true
    UseLoopCountLimit=false
   ) {module=SpawnBurst_Instantaneous001}
   /SetVariables(Emitter.RandomRange:Vector3f=&RandomRangeVector(Minimum=<0 0 0>, Maximum=<1 1 1>, Random Seed=2)) {isSetParameters=true, module=SetVariables_#1}
   /SetVariables(
    Emitter.Radius:NiagaraFloat=&self:StageSelection(
     Stage:NiagaraInt32=$User.Stage
     Stage 0:NiagaraFloat=0
     Stage 1:NiagaraFloat=375
     Stage 2:NiagaraFloat=300
     Stage 3:NiagaraFloat=500
     Stage 4:NiagaraFloat=500
     Stage 5:NiagaraFloat=580
     Stage 6:NiagaraFloat=650
    )
    Emitter.V1:Vector3f=&RandomVector(Vector Scale=1, Constrain to Hemisphere=true, Hemisphere Constraint Vector=$Emitter.RandomRange, Recalculate Random Each Loop=true)
    Emitter.V2:Vector3f=&RandomVector(Vector Scale=1, Constrain to Hemisphere=true, Hemisphere Constraint Vector=$Emitter.RandomRange, Recalculate Random Each Loop=true)
   ) {isSetParameters=true, module=SetVariables_#2}
   /SetVariables(
    Emitter.Radius:NiagaraFloat=&self:StageSelection(
     Stage:NiagaraInt32=$User.Stage
     Stage 0:NiagaraFloat=0
     Stage 1:NiagaraFloat=375
     Stage 2:NiagaraFloat=300
     Stage 3:NiagaraFloat=500
     Stage 4:NiagaraFloat=500
     Stage 5:NiagaraFloat=580
     Stage 6:NiagaraFloat=650
    )
    Emitter.V1:Vector3f=&RandomVector(Vector Scale=1)
   ) {isSetParameters=true}
   /SetVariables(Emitter.V2:Vector3f=&self:ScratchDynamicInput(
     Angle:NiagaraFloat=&Divide_Float(A=180, B=360)
     Rotation Axis:Vector3f=&RandomRangeVector(Minimum=<0 0 0>, Maximum=<1 1 1>)
     Vector To Rotate:Vector3f=$Emitter.V1
    )) {isSetParameters=true}
  ]
  particleSpawnScript [
   &InitializeParticle(
    Color=<rgba 0 1.9563 6 1>
    Lifetime=$Emitter.CurrentLoopDuration
    Ribbon Width=55
    Ribbon Width Mode=::NewEnumerator1
    Sprite Size Mode=::NewEnumerator3
    Uniform Sprite Size=100
    UsePositionOffset=false
    Write Material Random=true
   )
   &PartitionParticles(
    Normalized Index In Partition Includes One=true
    Number Of Partitions=3
    Particles Per Partition=1
    Partition Mode:ENiagara_PartitionMode=::NewEnumerator1
   )
   /SetVariables(Particles.RibbonLinkOrder:NiagaraFloat=&ReturnNormalizedExecIndex()) {isSetParameters=true, module=SetVariables_#3}
   /SetVariables(
    Particles.EndPosition:Vector3f=&Add_Vector(
     A=&SimulationPosition()
     B=&Multiply_VectorByFloat(Vector=$Emitter.V2, Float=$Emitter.Radius)
    )
    Particles.StartPosition:Vector3f=&Add_Vector(
     A=&SimulationPosition()
     B=&Multiply_VectorByFloat(Vector=$Emitter.V1, Float=$Emitter.Radius)
    )
   ) {isSetParameters=true, module=SetVariables_#4}
   /SetVariables(Particles.MiddlePoint:Vector3f=&Lerp_Vector(A=$Particles.StartPosition, B=$Particles.EndPosition, Alpha=0.5)) {isSetParameters=true, module=SetVariables_#5}
   &CameraOffset(Camera Offset Amount=30)
   &DynamicMaterialParameters(
    Index 0 Param 1=2
    Module.Param0WriteEnabled=true
    Module.Param1WriteEnabled=true
    Module.Param2WriteEnabled=true
    Module.Param3WriteEnabled=true
   )
  ]
  particleUpdateScript [
   &ParticleState()
   /SetVariables(Particles.FlareDistance:NiagaraFloat=&FloatFromCurve(
     FloatCurve=DataInterface() {propertyValues {Curve {keys [{arriveTangent=2.0921, interpMode=RCIM_Cubic, leaveTangent=2.0921, tangentMode=RCTM_User, time=0, value=0}, <at 1 1 RCIM_Cubic>]}}}
     Scale Curve=&RandomRangeFloat(Minimum=700, Maximum=400)
    )) {isSetParameters=true, module=SetVariables_#6}
   &self:BezierSpline(
    EndHandle:Vector3f=&Add_Vector(
     A=&Lerp_Vector(A=$Particles.MiddlePoint, B=$Particles.EndPosition, Alpha=0.5)
     B=&Multiply_VectorByFloat(
      Vector=&NormalizeVector(
       Vector To Normalize=&Subtract_Vector(A=$Particles.MiddlePoint, B=&SimulationPosition())
       Safe Normalize=true
       Safe Fallback Threshold=0
       Safe Fallback Vector=<0 0 0.001>
      )
      Float=$Particles.FlareDistance
     )
    )
    EndPosition:Vector3f=$Particles.EndPosition
    PositionAlongSpline:NiagaraFloat=&ReturnNormalizedExecIndex()
    StartHandle:Vector3f=&Add_Vector(
     A=&Lerp_Vector(A=$Particles.EndPosition, B=$Particles.MiddlePoint, Alpha=0.5)
     B=&Multiply_VectorByFloat(
      Vector=&NormalizeVector(
       Vector To Normalize=&Subtract_Vector(A=$Particles.MiddlePoint, B=&SimulationPosition())
       Safe Normalize=true
       Safe Fallback Threshold=0
       Safe Fallback Vector=<0 0 0.001>
      )
      Float=$Particles.FlareDistance
     )
    )
    StartPosition:Vector3f=$Particles.StartPosition
   ) {module=BezierSpline}
   /SetVariables(Particles.Position:Vector3f=$Particles.Spline_Position) {isSetParameters=true, module=SetVariables_#7}
   &Color(Color=&LerpLinearColors(StartColor=<rgba 0 0.019 1 1>, EndColor=<rgba 1 0.0334 0 1>, LerpFactor=$Particles.PartitionParticles.Normalized Index In Partition)) {enabled=false}
   /SetVariables(Particles.Position=&Lerp_Vector(A=$Particles.StartPosition, B=$Particles.EndPosition, Alpha=&ReturnNormalizedExecIndex())) {enabled=false, isSetParameters=true, module=SetVariables_#8}
   &ScaleColor(
    Scale Alpha=&FloatFromCurve(FloatCurve=DataInterface() {propertyValues {Curve {keys [<at 0 0 RCIM_Cubic>,<at 0.25 1 RCIM_Cubic>,<at 1 0 RCIM_Cubic>]}}})
    ScaleA=true
    ScaleRGB=true
   )
   /SetVariables(Particles.SpriteColor=&MakeLinearColorFromVectorAndFloat(
     "Vector (RGB)"=&MakeVectorFromLinearColorRGB(Color=<rgba 0 0.3261 1 1>)
     "Float (Alpha)"=&Multiply_Float(
      A=&MakeCustomFloatFromBool(
       Bool=&SetBoolByIntComparison(A=$Particles.PartitionParticles.Partition, B=1, Comparison Type=::NewEnumerator4)
       True Float=0
       False Float=1
      )
      B=&FloatFromCurve(FloatCurve=DataInterface() {propertyValues {Curve {keys [<at 0 0 RCIM_Cubic>,<at 0.25 1 RCIM_Cubic>,<at 1 0 RCIM_Cubic>]}}})
     )
    )) {isSetParameters=true, module=SetVariables_#9}
  ]
  rendererClasses [&NiagaraSpriteRendererProperties, &NiagaraRibbonRendererProperties]
  renderers [
   {rendererClass=&NiagaraSpriteRendererProperties}
   {rendererClass=&NiagaraRibbonRendererProperties, rendererIndex=1}
   {rendererClass=&NiagaraSpriteRendererProperties, rendererIndex=2}
   {rendererClass=&NiagaraSpriteRendererProperties, rendererIndex=3}
   {rendererClass=&NiagaraRibbonRendererProperties, rendererIndex=4}
   {rendererClass=&NiagaraSpriteRendererProperties, rendererIndex=5}
  ]
 }
 SolarFlares002 {
  emitterUpdateScript [
   &EmitterState(
    Enable Distance Culling=true
    Inactive Response=::NewEnumerator1
    Life Cycle Mode=::NewEnumerator1
    Loop Delay=&RandomRangeFloat(Minimum=0.1, Maximum=0.5)
    Loop Duration=&RandomRangeFloat(Minimum=2.5, Maximum=3.5)
    MaxDistance=21000
    Module.bUseMaxDistance=true
    Module.bUseMinDistance=false
    Recalculate Duration Each Loop=true
    Scalability Mode=::NewEnumerator1
    UseLoopDelay=true
   )
   &SpawnBurst_Instantaneous(
    Spawn Count=3
    Spawn Probability=&self:StageSelection(
     Stage:NiagaraInt32=$User.Stage
     Stage 0:NiagaraFloat=0
     Stage 1:NiagaraFloat=0
     Stage 2:NiagaraFloat=0
     Stage 3:NiagaraFloat=0
     Stage 4:NiagaraFloat=1
     Stage 5:NiagaraFloat=1
     Stage 6:NiagaraFloat=1
    )
    Use Spawn Probability=true
    UseLoopCountLimit=false
   ) {module=SpawnBurst_Instantaneous001}
   /SetVariables(Emitter.RandomRange:Vector3f=&RandomRangeVector(Minimum=<0 0 0>, Maximum=<1 1 1>, Random Seed=2)) {isSetParameters=true, module=SetVariables_#1}
   /SetVariables(
    Emitter.Radius:NiagaraFloat=&self:StageSelection(
     Stage:NiagaraInt32=$User.Stage
     Stage 0:NiagaraFloat=0
     Stage 1:NiagaraFloat=375
     Stage 2:NiagaraFloat=300
     Stage 3:NiagaraFloat=500
     Stage 4:NiagaraFloat=500
     Stage 5:NiagaraFloat=580
     Stage 6:NiagaraFloat=600
    )
    Emitter.V1:Vector3f=&RandomVector(Vector Scale=1, Constrain to Hemisphere=true, Hemisphere Constraint Vector=$Emitter.RandomRange, Recalculate Random Each Loop=true)
    Emitter.V2:Vector3f=&RandomVector(Vector Scale=1, Constrain to Hemisphere=true, Hemisphere Constraint Vector=$Emitter.RandomRange, Recalculate Random Each Loop=true)
   ) {isSetParameters=true, module=SetVariables_#2}
   /SetVariables(
    Emitter.Radius:NiagaraFloat=&self:StageSelection(
     Stage:NiagaraInt32=$User.Stage
     Stage 0:NiagaraFloat=0
     Stage 1:NiagaraFloat=375
     Stage 2:NiagaraFloat=300
     Stage 3:NiagaraFloat=500
     Stage 4:NiagaraFloat=500
     Stage 5:NiagaraFloat=580
     Stage 6:NiagaraFloat=650
    )
    Emitter.V1:Vector3f=&RandomVector(Vector Scale=1)
   ) {isSetParameters=true, module=SetVariables_#10}
   /SetVariables(Emitter.V2:Vector3f=&self:ScratchDynamicInput(
     Angle:NiagaraFloat=&Divide_Float(A=180, B=360)
     Rotation Axis:Vector3f=&RandomRangeVector(Minimum=<0 0 0>, Maximum=<1 1 1>)
     Vector To Rotate:Vector3f=$Emitter.V1
    )) {isSetParameters=true, module=SetVariables_#11}
  ]
  particleSpawnScript [
   &InitializeParticle(
    Color=<rgba 0 1.9563 6 1>
    Lifetime=$Emitter.CurrentLoopDuration
    Ribbon Width=55
    Ribbon Width Mode=::NewEnumerator1
    Sprite Size Mode=::NewEnumerator3
    Uniform Sprite Size=100
    UsePositionOffset=false
    Write Material Random=true
   )
   &PartitionParticles(
    Normalized Index In Partition Includes One=true
    Number Of Partitions=3
    Particles Per Partition=1
    Partition Mode:ENiagara_PartitionMode=::NewEnumerator1
   )
   /SetVariables(Particles.RibbonLinkOrder:NiagaraFloat=&ReturnNormalizedExecIndex()) {isSetParameters=true, module=SetVariables_#3}
   /SetVariables(
    Particles.EndPosition:Vector3f=&Add_Vector(
     A=&SimulationPosition()
     B=&Multiply_VectorByFloat(Vector=$Emitter.V2, Float=$Emitter.Radius)
    )
    Particles.StartPosition:Vector3f=&Add_Vector(
     A=&SimulationPosition()
     B=&Multiply_VectorByFloat(Vector=$Emitter.V1, Float=$Emitter.Radius)
    )
   ) {isSetParameters=true, module=SetVariables_#4}
   /SetVariables(Particles.MiddlePoint:Vector3f=&Lerp_Vector(A=$Particles.StartPosition, B=$Particles.EndPosition, Alpha=0.5)) {isSetParameters=true, module=SetVariables_#5}
   &CameraOffset(Camera Offset Amount=30)
   &DynamicMaterialParameters(
    Index 0 Param 1=2
    Module.Param0WriteEnabled=true
    Module.Param1WriteEnabled=true
    Module.Param2WriteEnabled=true
    Module.Param3WriteEnabled=true
   )
  ]
  particleUpdateScript [
   &ParticleState()
   /SetVariables(Particles.FlareDistance:NiagaraFloat=&FloatFromCurve(
     FloatCurve=DataInterface() {propertyValues {Curve {keys [{arriveTangent=2.0921, interpMode=RCIM_Cubic, leaveTangent=2.0921, tangentMode=RCTM_User, time=0, value=0}, <at 1 1 RCIM_Cubic>]}}}
     Scale Curve=&RandomRangeFloat(Minimum=700, Maximum=400)
    )) {isSetParameters=true, module=SetVariables_#6}
   &self:BezierSpline(
    EndHandle:Vector3f=&Add_Vector(
     A=&Lerp_Vector(A=$Particles.MiddlePoint, B=$Particles.EndPosition, Alpha=0.5)
     B=&Multiply_VectorByFloat(
      Vector=&NormalizeVector(
       Vector To Normalize=&Subtract_Vector(A=$Particles.MiddlePoint, B=&SimulationPosition())
       Safe Normalize=true
       Safe Fallback Threshold=0
       Safe Fallback Vector=<0 0 0.001>
      )
      Float=$Particles.FlareDistance
     )
    )
    EndPosition:Vector3f=$Particles.EndPosition
    PositionAlongSpline:NiagaraFloat=&ReturnNormalizedExecIndex()
    StartHandle:Vector3f=&Add_Vector(
     A=&Lerp_Vector(A=$Particles.EndPosition, B=$Particles.MiddlePoint, Alpha=0.5)
     B=&Multiply_VectorByFloat(
      Vector=&NormalizeVector(
       Vector To Normalize=&Subtract_Vector(A=$Particles.MiddlePoint, B=&SimulationPosition())
       Safe Normalize=true
       Safe Fallback Threshold=0
       Safe Fallback Vector=<0 0 0.001>
      )
      Float=$Particles.FlareDistance
     )
    )
    StartPosition:Vector3f=$Particles.StartPosition
   ) {module=BezierSpline}
   /SetVariables(Particles.Position:Vector3f=$Particles.Spline_Position) {isSetParameters=true, module=SetVariables_#7}
   &Color(Color=&LerpLinearColors(StartColor=<rgba 0 0.019 1 1>, EndColor=<rgba 1 0.0334 0 1>, LerpFactor=$Particles.PartitionParticles.Normalized Index In Partition)) {enabled=false}
   /SetVariables(Particles.Position=&Lerp_Vector(A=$Particles.StartPosition, B=$Particles.EndPosition, Alpha=&ReturnNormalizedExecIndex())) {enabled=false, isSetParameters=true, module=SetVariables_#8}
   &ScaleColor(
    Scale Alpha=&FloatFromCurve(FloatCurve=DataInterface() {propertyValues {Curve {keys [<at 0 0 RCIM_Cubic>,<at 0.25 1 RCIM_Cubic>,<at 1 0 RCIM_Cubic>]}}})
    ScaleA=true
    ScaleRGB=true
   )
   /SetVariables(Particles.SpriteColor=&MakeLinearColorFromVectorAndFloat(
     "Vector (RGB)"=&MakeVectorFromLinearColorRGB(Color=<rgba 0 0.3261 1 1>)
     "Float (Alpha)"=&Multiply_Float(
      A=&MakeCustomFloatFromBool(
       Bool=&SetBoolByIntComparison(A=$Particles.PartitionParticles.Partition, B=1, Comparison Type=::NewEnumerator4)
       True Float=0
       False Float=1
      )
      B=&FloatFromCurve(FloatCurve=DataInterface() {propertyValues {Curve {keys [<at 0 0 RCIM_Cubic>,<at 0.25 1 RCIM_Cubic>,<at 1 0 RCIM_Cubic>]}}})
     )
    )) {isSetParameters=true, module=SetVariables_#9}
  ]
  rendererClasses [&NiagaraSpriteRendererProperties, &NiagaraRibbonRendererProperties]
  renderers [
   {rendererClass=&NiagaraSpriteRendererProperties}
   {rendererClass=&NiagaraRibbonRendererProperties, rendererIndex=1}
   {rendererClass=&NiagaraSpriteRendererProperties, rendererIndex=2}
   {rendererClass=&NiagaraSpriteRendererProperties, rendererIndex=3}
   {rendererClass=&NiagaraRibbonRendererProperties, rendererIndex=4}
   {rendererClass=&NiagaraSpriteRendererProperties, rendererIndex=5}
  ]
 }
 SolarFlares003_Front {
  emitterUpdateScript [
   &EmitterState(
    Enable Distance Culling=true
    Inactive Response=::NewEnumerator1
    Life Cycle Mode=::NewEnumerator1
    Loop Delay=&RandomRangeFloat(Minimum=0.1, Maximum=1)
    Loop Duration=&RandomRangeFloat(Minimum=1.5, Maximum=2.2)
    MaxDistance=21000
    Module.bUseMaxDistance=true
    Module.bUseMinDistance=false
    Recalculate Duration Each Loop=true
    Scalability Mode=::NewEnumerator1
    UseLoopDelay=true
   )
   /SetVariables(Emitter.Cam Vec:Vector3f=&NormalizeVector(
     Vector To Normalize=&ReturnCameraVector(Camera Vectors=::NewEnumerator4, Coordinate Space=::Simulation, Origin Position=&SimulationPosition())
     Safe Normalize=true
     Safe Fallback Threshold=0
     Safe Fallback Vector=<0 0 0.001>
    )) {isSetParameters=true}
   &SpawnBurst_Instantaneous(
    Spawn Count=3
    Spawn Probability=&self:StageSelection(
     Stage:NiagaraInt32=$User.Stage
     Stage 0:NiagaraFloat=0
     Stage 1:NiagaraFloat=0
     Stage 2:NiagaraFloat=0
     Stage 3:NiagaraFloat=0
     Stage 4:NiagaraFloat=1
     Stage 5:NiagaraFloat=1
     Stage 6:NiagaraFloat=1
    )
    Use Spawn Probability=true
    UseLoopCountLimit=false
   ) {module=SpawnBurst_Instantaneous001}
   /SetVariables(Emitter.RandomRange:Vector3f=&RandomRangeVector(Minimum=<0 0 0>, Maximum=<1 1 1>, Random Seed=2)) {isSetParameters=true, module=SetVariables_#1}
   /SetVariables(
    Emitter.Radius:NiagaraFloat=&self:StageSelection(
     Stage:NiagaraInt32=$User.Stage
     Stage 0:NiagaraFloat=0
     Stage 1:NiagaraFloat=375
     Stage 2:NiagaraFloat=300
     Stage 3:NiagaraFloat=500
     Stage 4:NiagaraFloat=500
     Stage 5:NiagaraFloat=580
     Stage 6:NiagaraFloat=600
    )
    Emitter.V1:Vector3f=&self:ScratchDynamicInput(
     Angle:NiagaraFloat=&AngleConversion(
      Angle:NiagaraFloat=&RandomRangeFloat(Minimum=22, Maximum=45)
      Angle Input:ENiagara_AngleInput=::NewEnumerator0
      Angle Output:ENiagara_AngleInput=::NewEnumerator1
     )
     Rotation Axis:Vector3f=&self:ScratchDynamicInput_01(
      Default:Vector3f=&RandomRangeVector(Minimum=<0 0 0>, Maximum=<1 1 0>)
      Normal:Vector3f=$Emitter.Cam Vec
     )
     Vector To Rotate:Vector3f=$Emitter.Cam Vec
    )
    Emitter.V2:Vector3f=&RandomVector(Vector Scale=1)
   ) {isSetParameters=true, module=SetVariables_#2}
   /SetVariables(
    Emitter.Radius:NiagaraFloat=&self:StageSelection(
     Stage:NiagaraInt32=$User.Stage
     Stage 0:NiagaraFloat=0
     Stage 1:NiagaraFloat=375
     Stage 2:NiagaraFloat=300
     Stage 3:NiagaraFloat=500
     Stage 4:NiagaraFloat=500
     Stage 5:NiagaraFloat=580
     Stage 6:NiagaraFloat=650
    )
    Emitter.V1:Vector3f=&RandomVector(Vector Scale=1)
   ) {enabled=false, isSetParameters=true, module=SetVariables_#10}
   /SetVariables(Emitter.V2:Vector3f=&self:ScratchDynamicInput(
     Angle:NiagaraFloat=&Divide_Float(A=120, B=360)
     Rotation Axis:Vector3f=&RandomRangeVector(Minimum=<-1 -1 -1>, Maximum=<1 1 1>)
     Vector To Rotate:Vector3f=$Emitter.V1
    )) {isSetParameters=true, module=SetVariables_#11}
  ]
  particleSpawnScript [
   &InitializeParticle(
    Color=<rgba 0 1.9563 6 1>
    Lifetime=$Emitter.CurrentLoopDuration
    Ribbon Width=55
    Ribbon Width Mode=::NewEnumerator1
    Sprite Size Mode=::NewEnumerator3
    Uniform Sprite Size=100
    UsePositionOffset=false
    Write Material Random=true
   )
   &PartitionParticles(
    Normalized Index In Partition Includes One=true
    Number Of Partitions=3
    Particles Per Partition=1
    Partition Mode:ENiagara_PartitionMode=::NewEnumerator1
   )
   /SetVariables(Particles.RibbonLinkOrder:NiagaraFloat=&ReturnNormalizedExecIndex()) {isSetParameters=true, module=SetVariables_#3}
   /SetVariables(
    Particles.EndPosition:Vector3f=&Add_Vector(
     A=&SimulationPosition()
     B=&Multiply_VectorByFloat(Vector=$Emitter.V2, Float=$Emitter.Radius)
    )
    Particles.StartPosition:Vector3f=&Add_Vector(
     A=&SimulationPosition()
     B=&Multiply_VectorByFloat(Vector=$Emitter.V1, Float=$Emitter.Radius)
    )
   ) {isSetParameters=true, module=SetVariables_#4}
   /SetVariables(Particles.MiddlePoint:Vector3f=&Lerp_Vector(A=$Particles.StartPosition, B=$Particles.EndPosition, Alpha=0.5)) {isSetParameters=true, module=SetVariables_#5}
   &CameraOffset(Camera Offset Amount=30)
   &DynamicMaterialParameters(
    Index 0 Param 1=2
    Module.Param0WriteEnabled=true
    Module.Param1WriteEnabled=true
    Module.Param2WriteEnabled=true
    Module.Param3WriteEnabled=true
   )
  ]
  particleUpdateScript [
   &ParticleState()
   /SetVariables(Particles.FlareDistance:NiagaraFloat=&FloatFromCurve(
     FloatCurve=DataInterface() {propertyValues {Curve {keys [{arriveTangent=2.0921, interpMode=RCIM_Cubic, leaveTangent=2.0921, tangentMode=RCTM_User, time=0, value=0}, <at 1 1 RCIM_Cubic>]}}}
     Scale Curve=&RandomRangeFloat(Minimum=700, Maximum=800)
    )) {isSetParameters=true, module=SetVariables_#6}
   &self:BezierSpline(
    EndHandle:Vector3f=&Add_Vector(
     A=&Lerp_Vector(A=$Particles.MiddlePoint, B=$Particles.EndPosition, Alpha=0.5)
     B=&Multiply_VectorByFloat(
      Vector=&NormalizeVector(
       Vector To Normalize=&Subtract_Vector(A=$Particles.MiddlePoint, B=&SimulationPosition())
       Safe Normalize=true
       Safe Fallback Threshold=0
       Safe Fallback Vector=<0 0 0.001>
      )
      Float=$Particles.FlareDistance
     )
    )
    EndPosition:Vector3f=$Particles.EndPosition
    PositionAlongSpline:NiagaraFloat=&ReturnNormalizedExecIndex()
    StartHandle:Vector3f=&Add_Vector(
     A=&Lerp_Vector(A=$Particles.EndPosition, B=$Particles.MiddlePoint, Alpha=0.5)
     B=&Multiply_VectorByFloat(
      Vector=&NormalizeVector(
       Vector To Normalize=&Subtract_Vector(A=$Particles.MiddlePoint, B=&SimulationPosition())
       Safe Normalize=true
       Safe Fallback Threshold=0
       Safe Fallback Vector=<0 0 0.001>
      )
      Float=$Particles.FlareDistance
     )
    )
    StartPosition:Vector3f=$Particles.StartPosition
   ) {module=BezierSpline}
   /SetVariables(Particles.Position:Vector3f=$Particles.Spline_Position) {isSetParameters=true, module=SetVariables_#7}
   &Color(Color=&LerpLinearColors(StartColor=<rgba 0 0.019 1 1>, EndColor=<rgba 1 0.0334 0 1>, LerpFactor=$Particles.PartitionParticles.Normalized Index In Partition)) {enabled=false}
   /SetVariables(Particles.Position=&Lerp_Vector(A=$Particles.StartPosition, B=$Particles.EndPosition, Alpha=&ReturnNormalizedExecIndex())) {enabled=false, isSetParameters=true, module=SetVariables_#8}
   &ScaleColor(
    Scale Alpha=&FloatFromCurve(FloatCurve=DataInterface() {propertyValues {Curve {keys [<at 0 0 RCIM_Cubic>,<at 0.25 1 RCIM_Cubic>,<at 1 0 RCIM_Cubic>]}}})
    ScaleA=true
    ScaleRGB=true
   )
   /SetVariables(Particles.SpriteColor=&MakeLinearColorFromVectorAndFloat(
     "Vector (RGB)"=&MakeVectorFromLinearColorRGB(Color=<rgba 0 0.3261 1 1>)
     "Float (Alpha)"=&Multiply_Float(
      A=&MakeCustomFloatFromBool(
       Bool=&SetBoolByIntComparison(A=$Particles.PartitionParticles.Partition, B=1, Comparison Type=::NewEnumerator4)
       True Float=0
       False Float=1
      )
      B=&FloatFromCurve(FloatCurve=DataInterface() {propertyValues {Curve {keys [<at 0 0 RCIM_Cubic>,<at 0.25 1 RCIM_Cubic>,<at 1 0 RCIM_Cubic>]}}})
     )
    )) {isSetParameters=true, module=SetVariables_#9}
  ]
  rendererClasses [&NiagaraSpriteRendererProperties, &NiagaraRibbonRendererProperties]
  renderers [
   {rendererClass=&NiagaraSpriteRendererProperties}
   {rendererClass=&NiagaraRibbonRendererProperties, rendererIndex=1}
   {rendererClass=&NiagaraSpriteRendererProperties, rendererIndex=2}
   {rendererClass=&NiagaraSpriteRendererProperties, rendererIndex=3}
   {rendererClass=&NiagaraRibbonRendererProperties, rendererIndex=4}
   {rendererClass=&NiagaraSpriteRendererProperties, rendererIndex=5}
  ]
 }
}
summary {
 systemName=NS_Radish_ZeroPoint_Idle
 emitters [
  {emitterName=EnergyWaves, rendererClasses [&NiagaraMeshRendererProperties]}
  {emitterName=SolarFlares, rendererClasses [&NiagaraSpriteRendererProperties, &NiagaraRibbonRendererProperties]}
  {
   emitterName=SolarFlares001
   rendererClasses [&NiagaraSpriteRendererProperties, &NiagaraRibbonRendererProperties]
  }
  {
   bEnabled=false
   emitterName=RainbowCorona003
   rendererClasses [&NiagaraSpriteRendererProperties, &NiagaraMeshRendererProperties]
  }
  {
   bEnabled=false
   emitterName=RainbowCorona004
   rendererClasses [&NiagaraSpriteRendererProperties, &NiagaraMeshRendererProperties]
  }
  {
   bEnabled=false
   emitterName=RainbowCorona005
   rendererClasses [&NiagaraSpriteRendererProperties, &NiagaraMeshRendererProperties]
  }
  {emitterName=LightMesh, rendererClasses [&NiagaraSpriteRendererProperties, &NiagaraMeshRendererProperties]}
  {
   emitterName=SolarFlares002
   rendererClasses [&NiagaraSpriteRendererProperties, &NiagaraRibbonRendererProperties]
  }
  {
   emitterName=SolarFlares003_Front
   rendererClasses [&NiagaraSpriteRendererProperties, &NiagaraRibbonRendererProperties]
  }
  {emitterName=Empty, rendererClasses [&NiagaraSpriteRendererProperties]}
  {emitterName=Purple_Mesh_Glow_Burst, rendererClasses [&NiagaraMeshRendererProperties]}
 ]
 userVariables [
  User.ChargeUp {type=NiagaraFloat, defaultValue=1, description=""}
  User.RisingProcess {type=NiagaraFloat, defaultValue=1, description=""}
  User.SequenceRingIntensity {type=NiagaraFloat, defaultValue=1, description=""}
  User.SequenceRingSize {type=NiagaraFloat, defaultValue=0, description=""}
  User.Stage {type=NiagaraInt32, defaultValue=6}
 ]
}
